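# Translucent unlit Niagara particle material that renders the skull impact effect for the CoreStreet guitar cosmetic.
# ParticleRandom mirrors UVs horizontally on some particles, feeding both skull texture and SDF erosion mask.
# Particle color tints the skull through an emissive adjustment function.
# Opacity combines texture alpha, soft erosion driven by a dynamic parameter, and particle alpha, with depth fade skipped on low quality and mobile.
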Material /SparksCosmetics/Instrument/Guitar/CoreStreet/Effects/M_CoreStreet_Impact_Skull
graph {
 properties {
  bAutomaticallySetUsageInEditor=false
  bUsedWithNiagaraMeshParticles=true
  bUsedWithNiagaraSprites=true
  blendMode=BLEND_Translucent
  shadingModel=MSM_Unlit
  usedShadingModels=Unlit
 }
 nodes [
  n1 = TextureCoordinate()
  n2 = MaterialFunctionCall(Float2=@n1) {materialFunction=&BreakOutFloat2Components}
  n3 = MaterialFunctionCall(X=OneMinus(None=@n2.R), Y=@n2.G) {materialFunction=&MakeFloat2}
  n4 = Step(X=ParticleRandom()) {constY=0.5}
  n5 = TextureSampleParameter2D(UVs=NamedRerouteDeclaration(None=LinearInterpolate(A=@n1, B=@n3.Result, Alpha=@n4)) {name=UVs, nodeColor=<rgba 0 1 0.2031 1>}) {parameterName=Skull Texture, texture=&/T_CoreStreet_Skull}
  n6 = ParticleColor()
  MaterialFunctionCall(Emissive=Multiply(A=@n5.RGB, B=@n6.RGB)) {materialFunction=&MF_FX_EmissiveAdjustments}
  n7 = TextureSampleParameter2D(UVs=NamedRerouteUsage()) {
   parameterName=Erosion Texture
   samplerType=SAMPLERTYPE_LinearGrayscale
   texture=&/T_CoreStreet_Skull_SDF_Mask
  }
  n8 = DynamicParameter() {defaultValue=<rgba 0 1 1 1>, paramNames [Erosion Amount, Param2, Param3, Param4]}
  n9 = Constant() {desc=Softness, r=0.1}
  n10 = Multiply(
   A=Multiply(
    A=@n5.A
    B=MaterialFunctionCall!(Gradient=@n7.R, Erosion Amount=@n8.Erosion Amount, Softness=@n9) {materialFunction=/BRCosmetics/Athena/Effects/Particles/Skydiving/NS_Contrail_DarkAndLight/Functions/MF_Erosion}.Default Range
   )
   B=@n6.A
  )
  n11 = DepthFade(Opacity=@n10, FadeDistance=ScalarParameter() {defaultValue=10, group=Opacity, parameterName=DepthFade, sortPriority=0})
  n12 = QualitySwitch(Default=@n11, Low=@n10, High=@n11, Medium=@n11, Epic=@n11)
  Reroute(None=Saturate(None=ShadingPathSwitch(Default=@n12, Mobile=@n10))) {desc=Opacity Out}
 ]
}
